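# Physics asset for dynamic secondary motion on the male medium Assassin X cosmetic body's collar and chest flaps, pant buckles, shirt panels and straps.
# Pelvis, thigh and calf capsules form the body; MSTR root bodies join pelvis or thighs through free constraints.
# Four-segment box strap chains hang off locked roots on side shirt panels, so they inherit panel motion.
# Collar and buckles lock twist and one swing, giving hinge-like movement.
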PhysicsAsset /BRCosmetics/Characters/Player/Male/Medium/Bodies/M_MED_Assassin_X/Meshes/M_MED_Assassin_X_dyn
bodies {
 calf_l [{shapeType=Capsule, length=25.1479, radius=8.2739, center=<-18.7962 0 0>, rotation=<rot 90 0 0>}]
 calf_r [{shapeType=Capsule, length=25.1479, radius=8.2739, center=<18.7962 0 0>, rotation=<rot 90 0 0>}]
 dyn_chest_flap_A_01 [{shapeType=Capsule, length=3, radius=1, center=<1 0 0>, rotation=<rot 90 0 0>}]
 dyn_chest_flap_A_MSTR [{shapeType=Capsule, length=0.5, radius=0.5}]
 dyn_chest_flap_B_01 [{shapeType=Capsule, length=3, radius=1, center=<1 0 0>, rotation=<rot 90 0 0>}]
 dyn_chest_flap_B_MSTR [{shapeType=Capsule, length=0.5, radius=0.5}]
 dyn_collar_flap_01 [{shapeType=Capsule, length=3, radius=1, center=<1 0 0>, rotation=<rot 90 0 0>}]
 dyn_collar_flap_MSTR [{shapeType=Capsule, length=0.5, radius=0.5}]
 dyn_pant_buckle_l_01 [{shapeType=Capsule, length=3, radius=1, center=<1.2368 0 0.0001>, rotation=<rot 90 0 0>}]
 dyn_pant_buckle_l_MSTR [{shapeType=Capsule, length=0.5, radius=0.5}]
 dyn_pant_buckle_r_01 [{shapeType=Capsule, length=3, radius=1, center=<1.2368 0 0.0001>, rotation=<rot 90 0 0>}]
 dyn_pant_buckle_r_MSTR [{shapeType=Capsule, length=0.5, radius=0.5}]
 dyn_shirt_bk_l_01 [{shapeType=Box, extentX=7.5625, extentY=10.9375, extentZ=1.25, center=<2.8312 0 0.0001>}]
 dyn_shirt_bk_l_MSTR [{shapeType=Capsule, length=0.5, radius=0.5}]
 dyn_shirt_bk_mid_01 [{shapeType=Box, extentX=8, extentY=14.4375, extentZ=1.25, center=<3.016 0 0>}]
 dyn_shirt_bk_mid_MSTR [{shapeType=Capsule, length=0.5, radius=0.5}]
 dyn_shirt_bk_r_01 [{shapeType=Box, extentX=7.5625, extentY=10.9375, extentZ=1.25, center=<2.8312 0 0.0001>}]
 dyn_shirt_bk_r_MSTR [{shapeType=Capsule, length=0.5, radius=0.5}]
 dyn_shirt_fr_l_01 [{shapeType=Box, extentX=14.9375, extentY=7.5, extentZ=1.25, center=<6.5284 0 0.0001>}]
 dyn_shirt_fr_l_MSTR [{shapeType=Capsule, length=0.5, radius=0.5}]
 dyn_shirt_fr_mid_01 [{shapeType=Box, extentX=8.6875, extentY=4.4375, extentZ=1.25, center=<4.0762 0 0.0001>}]
 dyn_shirt_fr_mid_02 [{shapeType=Box, extentX=7.8125, extentY=4.4375, extentZ=1.25, center=<4.2472 0 0>}]
 dyn_shirt_fr_mid_MSTR [{shapeType=Capsule, length=0.5, radius=0.5}]
 dyn_shirt_fr_r_01 [{shapeType=Box, extentX=14.9375, extentY=7.5, extentZ=1.25, center=<6.5284 0 0.0001>}]
 dyn_shirt_fr_r_MSTR [{shapeType=Capsule, length=0.5, radius=0.5}]
 dyn_strap_bk_l_01 [{shapeType=Box, extentX=7, extentY=2.625, extentZ=1.25, center=<3.3352 0 0>}]
 dyn_strap_bk_l_02 [{shapeType=Box, extentX=7.125, extentY=2.625, extentZ=1.25, center=<3.6906 0 0>}]
 dyn_strap_bk_l_03 [{shapeType=Box, extentX=5.375, extentY=2.625, extentZ=1.25, center=<2.5858 0 0>}]
 dyn_strap_bk_l_04 [{shapeType=Box, extentX=3.5, extentY=2.625, extentZ=1.25, center=<1.3901 0 0>}]
 dyn_strap_bk_l_MSTR [{shapeType=Capsule, length=0.505, radius=0.505}]
 dyn_strap_bk_r_01 [{shapeType=Box, extentX=7, extentY=2.625, extentZ=1.25, center=<3.3352 0 0>}]
 dyn_strap_bk_r_02 [{shapeType=Box, extentX=7.125, extentY=2.625, extentZ=1.25, center=<3.6906 0 0>}]
 dyn_strap_bk_r_03 [{shapeType=Box, extentX=5.375, extentY=2.625, extentZ=1.25, center=<2.5858 0 0>}]
 dyn_strap_bk_r_04 [{shapeType=Box, extentX=3.5, extentY=2.625, extentZ=1.25, center=<1.3901 0 0>}]
 dyn_strap_bk_r_MSTR [{shapeType=Capsule, length=0.505, radius=0.505}]
 dyn_strap_fr_l_01 [{shapeType=Box, extentX=7, extentY=3, extentZ=1.25, center=<3.5297 0 0>}]
 dyn_strap_fr_l_02 [{shapeType=Box, extentX=6.8125, extentY=3, extentZ=1.25, center=<3.6562 0 0.0002>}]
 dyn_strap_fr_l_03 [{shapeType=Box, extentX=5.25, extentY=3, extentZ=1.25, center=<2.6426 0 0>}]
 dyn_strap_fr_l_04 [{shapeType=Box, extentX=3.75, extentY=3, extentZ=1.25, center=<2.1256 0 0.0001>}]
 dyn_strap_fr_l_MSTR [{shapeType=Capsule, length=0.505, radius=0.505}]
 dyn_strap_fr_r_01 [{shapeType=Box, extentX=7, extentY=3, extentZ=1.25, center=<3.5297 0 0>}]
 dyn_strap_fr_r_02 [{shapeType=Box, extentX=6.8125, extentY=3, extentZ=1.25, center=<3.6562 0 0.0002>}]
 dyn_strap_fr_r_03 [{shapeType=Box, extentX=5.25, extentY=3, extentZ=1.25, center=<2.6426 0 0>}]
 dyn_strap_fr_r_04 [{shapeType=Box, extentX=3.75, extentY=3, extentZ=1.25, center=<2.1256 0 0.0001>}]
 dyn_strap_fr_r_MSTR [{shapeType=Capsule, length=0.505, radius=0.505}]
 pelvis [{shapeType=Capsule, length=8.4006, radius=11.0525, center=<-2.5155 -0.6795 0>}]
 thigh_l [{shapeType=Capsule, length=25.1479, radius=8.2739, center=<-18.8012 0.0808 0>, rotation=<rot 90 17.354 13.8195>}]
 thigh_r [{shapeType=Capsule, length=25.1479, radius=8.2739, center=<18.8012 -0.0808 0>, rotation=<rot 90 17.354 13.8195>}]
}
constraints [
 {
  bone1Name=dyn_collar_flap_MSTR
  bone2Name=pelvis
  swing1LimitDegrees=45
  swing1Motion=Free
  swing2LimitDegrees=45
  swing2Motion=Free
  twistLimitDegrees=45
  twistMotion=Free
 }
 {
  bone1Name=dyn_collar_flap_01
  bone2Name=dyn_collar_flap_MSTR
  swing1LimitDegrees=45
  swing1Motion=Locked
  swing2LimitDegrees=45
  twistLimitDegrees=45
  twistMotion=Locked
 }
 {
  bone1Name=dyn_chest_flap_A_MSTR
  bone2Name=pelvis
  swing1LimitDegrees=45
  swing1Motion=Free
  swing2LimitDegrees=45
  swing2Motion=Free
  twistLimitDegrees=45
  twistMotion=Free
 }
 {bone1Name=dyn_chest_flap_A_01, bone2Name=dyn_chest_flap_A_MSTR, swing1LimitDegrees=20, swing2LimitDegrees=45, twistLimitDegrees=5}
 {
  bone1Name=dyn_chest_flap_B_MSTR
  bone2Name=pelvis
  swing1LimitDegrees=45
  swing1Motion=Free
  swing2LimitDegrees=45
  swing2Motion=Free
  twistLimitDegrees=45
  twistMotion=Free
 }
 {bone1Name=dyn_chest_flap_B_01, bone2Name=dyn_chest_flap_B_MSTR, swing1LimitDegrees=20, swing2LimitDegrees=45, twistLimitDegrees=5}
 {
  bone1Name=thigh_l
  bone2Name=pelvis
  swing1LimitDegrees=45
  swing1Motion=Free
  swing2LimitDegrees=45
  swing2Motion=Free
  twistLimitDegrees=45
  twistMotion=Free
 }
 {
  bone1Name=calf_l
  bone2Name=thigh_l
  swing1LimitDegrees=45
  swing1Motion=Free
  swing2LimitDegrees=45
  swing2Motion=Free
  twistLimitDegrees=45
  twistMotion=Free
 }
 {
  bone1Name=dyn_pant_buckle_l_MSTR
  bone2Name=thigh_l
  swing1LimitDegrees=45
  swing1Motion=Free
  swing2LimitDegrees=45
  swing2Motion=Free
  twistLimitDegrees=45
  twistMotion=Free
 }
 {
  bone1Name=dyn_pant_buckle_l_01
  bone2Name=dyn_pant_buckle_l_MSTR
  swing1LimitDegrees=45
  swing1Motion=Locked
  swing2LimitDegrees=45
  twistLimitDegrees=45
  twistMotion=Locked
 }
 {
  bone1Name=thigh_r
  bone2Name=pelvis
  swing1LimitDegrees=45
  swing1Motion=Free
  swing2LimitDegrees=45
  swing2Motion=Free
  twistLimitDegrees=45
  twistMotion=Free
 }
 {
  bone1Name=calf_r
  bone2Name=thigh_r
  swing1LimitDegrees=45
  swing1Motion=Free
  swing2LimitDegrees=45
  swing2Motion=Free
  twistLimitDegrees=45
  twistMotion=Free
 }
 {
  bone1Name=dyn_pant_buckle_r_MSTR
  bone2Name=thigh_r
  swing1LimitDegrees=45
  swing1Motion=Free
  swing2LimitDegrees=45
  swing2Motion=Free
  twistLimitDegrees=45
  twistMotion=Free
 }
 {
  bone1Name=dyn_pant_buckle_r_01
  bone2Name=dyn_pant_buckle_r_MSTR
  swing1LimitDegrees=45
  swing1Motion=Locked
  swing2LimitDegrees=45
  twistLimitDegrees=45
  twistMotion=Locked
 }
 {
  bone1Name=dyn_shirt_fr_mid_MSTR
  bone2Name=pelvis
  swing1LimitDegrees=45
  swing1Motion=Free
  swing2LimitDegrees=45
  swing2Motion=Free
  twistLimitDegrees=45
  twistMotion=Free
 }
 {bone1Name=dyn_shirt_fr_mid_01, bone2Name=dyn_shirt_fr_mid_MSTR, swing1LimitDegrees=45, swing2LimitDegrees=45, twistLimitDegrees=5}
 {bone1Name=dyn_shirt_fr_mid_02, bone2Name=dyn_shirt_fr_mid_01, swing1LimitDegrees=45, swing2LimitDegrees=45, twistLimitDegrees=5}
 {
  bone1Name=dyn_shirt_fr_l_MSTR
  bone2Name=pelvis
  swing1LimitDegrees=45
  swing1Motion=Free
  swing2LimitDegrees=45
  swing2Motion=Free
  twistLimitDegrees=45
  twistMotion=Free
 }
 {bone1Name=dyn_shirt_fr_l_01, bone2Name=dyn_shirt_fr_l_MSTR, swing1LimitDegrees=45, swing2LimitDegrees=45, twistLimitDegrees=5}
 {
  bone1Name=dyn_strap_fr_l_MSTR
  bone2Name=dyn_shirt_fr_l_01
  swing1LimitDegrees=45
  swing1Motion=Locked
  swing2LimitDegrees=45
  swing2Motion=Locked
  twistLimitDegrees=45
  twistMotion=Locked
 }
 {bone1Name=dyn_strap_fr_l_01, bone2Name=dyn_strap_fr_l_MSTR, swing1LimitDegrees=45, swing2LimitDegrees=45, twistLimitDegrees=5}
 {bone1Name=dyn_strap_fr_l_02, bone2Name=dyn_strap_fr_l_01, swing1LimitDegrees=45, swing2LimitDegrees=45, twistLimitDegrees=5}
 {bone1Name=dyn_strap_fr_l_03, bone2Name=dyn_strap_fr_l_02, swing1LimitDegrees=45, swing2LimitDegrees=45, twistLimitDegrees=5}
 {bone1Name=dyn_strap_fr_l_04, bone2Name=dyn_strap_fr_l_03, swing1LimitDegrees=45, swing2LimitDegrees=45, twistLimitDegrees=5}
 {
  bone1Name=dyn_shirt_bk_mid_MSTR
  bone2Name=pelvis
  swing1LimitDegrees=45
  swing1Motion=Free
  swing2LimitDegrees=45
  swing2Motion=Free
  twistLimitDegrees=45
  twistMotion=Free
 }
 {bone1Name=dyn_shirt_bk_mid_01, bone2Name=dyn_shirt_bk_mid_MSTR, swing1LimitDegrees=45, swing2LimitDegrees=45, twistLimitDegrees=5}
 {
  bone1Name=dyn_shirt_bk_l_MSTR
  bone2Name=pelvis
  swing1LimitDegrees=45
  swing1Motion=Free
  swing2LimitDegrees=45
  swing2Motion=Free
  twistLimitDegrees=45
  twistMotion=Free
 }
 {bone1Name=dyn_shirt_bk_l_01, bone2Name=dyn_shirt_bk_l_MSTR, swing1LimitDegrees=45, swing2LimitDegrees=45, twistLimitDegrees=5}
 {
  bone1Name=dyn_strap_bk_l_MSTR
  bone2Name=dyn_shirt_bk_l_01
  swing1LimitDegrees=45
  swing1Motion=Locked
  swing2LimitDegrees=45
  swing2Motion=Locked
  twistLimitDegrees=45
  twistMotion=Locked
 }
 {bone1Name=dyn_strap_bk_l_01, bone2Name=dyn_strap_bk_l_MSTR, swing1LimitDegrees=45, swing2LimitDegrees=45, twistLimitDegrees=5}
 {bone1Name=dyn_strap_bk_l_02, bone2Name=dyn_strap_bk_l_01, swing1LimitDegrees=45, swing2LimitDegrees=45, twistLimitDegrees=5}
 {bone1Name=dyn_strap_bk_l_03, bone2Name=dyn_strap_bk_l_02, swing1LimitDegrees=45, swing2LimitDegrees=45, twistLimitDegrees=5}
 {bone1Name=dyn_strap_bk_l_04, bone2Name=dyn_strap_bk_l_03, swing1LimitDegrees=45, swing2LimitDegrees=45, twistLimitDegrees=5}
 {
  bone1Name=dyn_shirt_bk_r_MSTR
  bone2Name=pelvis
  swing1LimitDegrees=45
  swing1Motion=Free
  swing2LimitDegrees=45
  swing2Motion=Free
  twistLimitDegrees=45
  twistMotion=Free
 }
 {bone1Name=dyn_shirt_bk_r_01, bone2Name=dyn_shirt_bk_r_MSTR, swing1LimitDegrees=45, swing2LimitDegrees=45, twistLimitDegrees=5}
 {
  bone1Name=dyn_strap_bk_r_MSTR
  bone2Name=dyn_shirt_bk_r_01
  swing1LimitDegrees=45
  swing1Motion=Locked
  swing2LimitDegrees=45
  swing2Motion=Locked
  twistLimitDegrees=45
  twistMotion=Locked
 }
 {bone1Name=dyn_strap_bk_r_01, bone2Name=dyn_strap_bk_r_MSTR, swing1LimitDegrees=45, swing2LimitDegrees=45, twistLimitDegrees=5}
 {bone1Name=dyn_strap_bk_r_02, bone2Name=dyn_strap_bk_r_01, swing1LimitDegrees=45, swing2LimitDegrees=45, twistLimitDegrees=5}
 {bone1Name=dyn_strap_bk_r_03, bone2Name=dyn_strap_bk_r_02, swing1LimitDegrees=45, swing2LimitDegrees=45, twistLimitDegrees=5}
 {bone1Name=dyn_strap_bk_r_04, bone2Name=dyn_strap_bk_r_03, swing1LimitDegrees=45, swing2LimitDegrees=45, twistLimitDegrees=5}
 {
  bone1Name=dyn_shirt_fr_r_MSTR
  bone2Name=pelvis
  swing1LimitDegrees=45
  swing1Motion=Free
  swing2LimitDegrees=45
  swing2Motion=Free
  twistLimitDegrees=45
  twistMotion=Free
 }
 {bone1Name=dyn_shirt_fr_r_01, bone2Name=dyn_shirt_fr_r_MSTR, swing1LimitDegrees=45, swing2LimitDegrees=45, twistLimitDegrees=5}
 {
  bone1Name=dyn_strap_fr_r_MSTR
  bone2Name=dyn_shirt_fr_r_01
  swing1LimitDegrees=45
  swing1Motion=Locked
  swing2LimitDegrees=45
  swing2Motion=Locked
  twistLimitDegrees=45
  twistMotion=Locked
 }
 {bone1Name=dyn_strap_fr_r_01, bone2Name=dyn_strap_fr_r_MSTR, swing1LimitDegrees=45, swing2LimitDegrees=45, twistLimitDegrees=5}
 {bone1Name=dyn_strap_fr_r_02, bone2Name=dyn_strap_fr_r_01, swing1LimitDegrees=45, swing2LimitDegrees=45, twistLimitDegrees=5}
 {bone1Name=dyn_strap_fr_r_03, bone2Name=dyn_strap_fr_r_02, swing1LimitDegrees=45, swing2LimitDegrees=45, twistLimitDegrees=5}
 {bone1Name=dyn_strap_fr_r_04, bone2Name=dyn_strap_fr_r_03, swing1LimitDegrees=45, swing2LimitDegrees=45, twistLimitDegrees=5}
]
