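# Animation Blueprint for the male medium Assassin X body cosmetic, adding secondary cloth motion to the pose copied from the attached parent mesh.
# Additive bone offsets lift front shirt bones using thigh-rotation and crouch alphas, then weighted constraints blend chest flaps, collar and buckles between body bones.
# A spine-relative RigidBody node simulates the dynamics physics asset, with gravity weakened or reversed by vertical velocity during skydiving until the parachute opens.
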AnimBlueprint /BRCosmetics/Characters/Player/Male/Medium/Bodies/M_MED_Assassin_X/Meshes/M_MED_Assassin_X_AnimBP
target_skeleton=&/M_MED_Assassin_X_Skeleton
graphs [
 AnimGraph {nodes [
   n1 = "Transform (Modify) Bone - Bone: dyn_shirt_fr_l_MSTR"(
    ComponentPose=Local To Component(LocalPose=Copy Pose From Mesh() {bUseAttachedParent=true})
    Alpha=Get thigh_l_rot_alpha() -> (thigh_l_rot_alpha:double)
    Translation="-0.908,-1.445,8.770"
    Rotation="0.000,0.761,-77.820"
    Scale="1,1,1"
   ) {
    rotationMode=BMM_Additive
    translationMode=BMM_Additive
    alphaBoolBlend {blendOption=Linear}
    alphaScaleBiasClamp {bInterpResult=true, interpSpeedDecreasing=15, interpSpeedIncreasing=15}
    boneToModify {boneName=dyn_shirt_fr_l_MSTR}
    rotation {roll=-77.8198, yaw=0.761}
    translation=<-0.9076 -1.445 8.77>
   }
   n2 = "Transform (Modify) Bone - Bone: dyn_shirt_fr_mid_MSTR"(ComponentPose=@n1, Alpha=Get crouch alpha() -> (crouch alpha:double), Translation="0.000,-0.279,12.082", Rotation="0,0,0", Scale="1,1,1") {
    rotationMode=BMM_Additive
    translationMode=BMM_Additive
    alphaBoolBlend {blendOption=Linear}
    alphaScaleBiasClamp {bInterpResult=true, interpSpeedDecreasing=15, interpSpeedIncreasing=15}
    boneToModify {boneName=dyn_shirt_fr_mid_MSTR}
    translation {y=-0.279, z=12.082}
   }
   n3 = "Transform (Modify) Bone - Bone: dyn_shirt_fr_r_MSTR"(
    ComponentPose=@n2
    Alpha=Get thigh_r_rot_alpha() -> (thigh_r_rot_alpha:double)
    Translation="0.426,5.404,8.937"
    Rotation="0.000,0.000,-60.000"
    Scale="1,1,1"
   ) {
    rotationMode=BMM_Additive
    translationMode=BMM_Additive
    alphaBoolBlend {blendOption=Linear}
    alphaScaleBiasClamp {bInterpResult=true, interpSpeedDecreasing=15, interpSpeedIncreasing=15}
    boneToModify {boneName=dyn_shirt_fr_r_MSTR}
    rotation {roll=-59.9999}
    translation=<0.426 5.4044 8.937>
   }
   n4 = "Constraint - dyn_chest_flap_A_MSTR (PPPP)"!(
    ComponentPose=@n3
    Alpha=1
    ConstraintWeights_0=0.1372
    ConstraintWeights_1:float=0.2834
    ConstraintWeights_2:float=0.4713
    ConstraintWeights_3:float=0.108
   ) -> (Pose) {
    alphaBoolBlend {blendOption=Linear}
    boneToModify {boneName=dyn_chest_flap_A_MSTR}
    constraintSetup [{targetBone {boneName=spine_02}}, {targetBone {boneName=spine_03}}, {targetBone {boneName=spine_04}}, {targetBone {boneName=spine_05}}]
    constraintWeights [1,1,1,1]
   }
   n5 = "Constraint - dyn_chest_flap_B_MSTR (PPPPPP)"!(
    ComponentPose=@n4
    Alpha=1
    ConstraintWeights_0=0.0425
    ConstraintWeights_1:float=0.2606
    ConstraintWeights_2:float=0.3444
    ConstraintWeights_3:float=0.288
    ConstraintWeights_4:float=0.0537
    ConstraintWeights_5:float=0.0107
   ) -> (Pose) {
    alphaBoolBlend {blendOption=Linear}
    boneToModify {boneName=dyn_chest_flap_B_MSTR}
    constraintSetup [
     {targetBone {boneName=spine_01}}
     {targetBone {boneName=spine_02}}
     {targetBone {boneName=spine_03}}
     {targetBone {boneName=spine_04}}
     {targetBone {boneName=spine_05}}
     {targetBone {boneName=pelvis}}
    ]
    constraintWeights [1,1,1,1,1,1]
   }
   n6 = "Constraint - dyn_collar_flap_MSTR (PPPP)"!(
    ComponentPose=@n5
    Alpha=1
    ConstraintWeights_0=0.0056
    ConstraintWeights_1:float=0.1023
    ConstraintWeights_2:float=0.4024
    ConstraintWeights_3:float=0.4897
   ) -> (Pose) {
    alphaBoolBlend {blendOption=Linear}
    boneToModify {boneName=dyn_collar_flap_MSTR}
    constraintSetup [{targetBone {boneName=spine_02}}, {targetBone {boneName=spine_03}}, {targetBone {boneName=spine_04}}, {targetBone {boneName=spine_05}}]
    constraintWeights [1,1,1,1]
   }
   n7 = "Constraint - dyn_pant_buckle_l_MSTR (PPP)"!(ComponentPose=@n6, Alpha=1, ConstraintWeights_0=0.133, ConstraintWeights_1:float=0.8268, ConstraintWeights_2:float=0.0402) -> (Pose) {
    alphaBoolBlend {blendOption=Linear}
    boneToModify {boneName=dyn_pant_buckle_l_MSTR}
    constraintSetup [{targetBone {boneName=pelvis}}, {targetBone {boneName=thigh_twist_01_l}}, {targetBone {boneName=thigh_l}}]
    constraintWeights [1,1,1]
   }
   n8 = "Constraint - dyn_pant_buckle_r_MSTR (PPP)"!(ComponentPose=@n7, Alpha=1, ConstraintWeights_0=0.03, ConstraintWeights_1:float=0.1729, ConstraintWeights_2:float=0.7971) -> (Pose) {
    alphaBoolBlend {blendOption=Linear}
    boneToModify {boneName=dyn_pant_buckle_r_MSTR}
    constraintSetup [{targetBone {boneName=thigh_r}}, {targetBone {boneName=pelvis}}, {targetBone {boneName=thigh_twist_01_r}}]
    constraintWeights [1,1,1]
   }
   n9 = RigidBody!(ComponentPose=@n8, Alpha, OverrideWorldGravity:Vector=Get gravity_override() -> (gravity_override:Vector), ExternalForce) -> (Pose) {
    bClampLinearTranslationLimitToRefPose=true
    bOverrideWorldGravity=true
    overridePhysicsAsset=&/M_MED_Assassin_X_dyn
    simulationSpace=BaseBoneSpace
    alphaBoolBlend {blendOption=Linear}
    baseBoneRef {boneName=spine_01}
    componentAppliedLinearAccClamp=<3000 3000 3000>
    componentLinearAccScale=<0.4 0.3 0.3>
    componentLinearVelScale=<0.4 0.3 0.3>
    overrideWorldGravity {z=-980}
   }
   Output Pose(Result=Component To Local(ComponentPose=@n9)) {layerGroup=None}
  ]}
 EventGraph {nodes [
   Event BlueprintUpdateAnimation()
   > n1 = Sequence()
   n2 = Get Skel_Mesh()
   > Set thigh_l_rot_alpha(thigh_l_rot_alpha:double=GetAlphaFromReferenceTransformDelta(SkeletalMeshComponent=@n2, BoneName=thigh_l, BaseName=pelvis, InputRange=(X=-50,Y=-90), TwistAxis="0,0,1")) -> (Output_Get:double)
   > Set thigh_r_rot_alpha(thigh_r_rot_alpha:double=GetAlphaFromReferenceTransformDelta(SkeletalMeshComponent=@n2, BoneName=thigh_r, BaseName=pelvis, InputRange=(X=-50,Y=-90), TwistAxis="0,0,1")) -> (Output_Get:double)
   n3 = Branch(execute=@n1.then_1, Condition=Get bIsSkyDiving())
   > n4 = Branch(Condition=Get bIsParachuteOpened())
   Set gravity_override(execute=[@n3.else, @n4.then], gravity_override="0, 0,-980")
   n5 = Branch(execute=@n4.else, Condition="float > float"(A=Get PawnVelocityZ(), B=50) {is=title K2Node_CallFunction})
   > Set gravity_override(gravity_override="0, 0,200")
   Set gravity_override(execute=@n5.else, gravity_override="0, 0,-200")
   Event OnBeginCrouching()
   > Set crouch alpha(crouch alpha:double=1) -> (Output_Get:double)
   Event OnEndCrouching()
   > Set crouch alpha(crouch alpha=0)
   Event BlueprintInitializeAnimation()
   > Set Skel_Mesh(Skel_Mesh=GetOwnerSkeletalMeshComponent())
  ]}
]
variables [
 crouch alpha {type=number}
 gravity_override {type=Vector}
 Skel_Mesh {type=SkeletalMeshComponent}
 thigh_l_rot_alpha {type=number}
 thigh_r_rot_alpha {type=number}
]
